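# Physics asset for the SameRabbit MiniTentacles backpack cosmetic, defining collision bodies and joints for its skeletal mesh.
# Each body is a single capsule: small ones on root, pack and body, and fitted, rotated ones on four dyn_spike bones.
# Locked angular constraints chain body to pack to root as a rigid base.
# Each spike joins directly to body with narrow swing and minimal twist limits, allowing only a slight wobble.
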PhysicsAsset /BRCosmetics/Accessories/FORT_Backpacks/Backpack_F_MED_SameRabbit/Meshes/F_MED_SameRabbit_MiniTentacles_Pack_Physics
bodies {
 body [{shapeType=Capsule, length=1, radius=1}]
 dyn_spike_1 [{
   shapeType=Capsule
   length=2.1269
   radius=1.7443
   shapeName=dyn_spike_1_capsule
   center=<2.8089 -0.0338 -0.1486>
   rotation=<rot 0 -88.119 -89.545>
  }]
 dyn_spike_2 [{
   shapeType=Capsule
   length=2.1269
   radius=1.7468
   shapeName=dyn_spike_2_capsule
   center=<2.8107 -0.0389 0.1093>
   rotation=<rot 0 -88.1892 -89.6848>
  }]
 dyn_spike_3 [{
   shapeType=Capsule
   length=2.3559
   radius=1.6662
   shapeName=dyn_spike_3_capsule
   center=<3.3363 0.0093 0.1814>
   rotation=<rot 0 -94.2014 -86.7117>
  }]
 dyn_spike_4 [{
   shapeType=Capsule
   length=2.3559
   radius=1.6659
   shapeName=dyn_spike_4_capsule
   center=<3.3431 0.0107 -0.1781>
   rotation=<rot 0 85.7665 -86.8034>
  }]
 pack [{shapeType=Capsule, length=0.505, radius=0.505, shapeName=pack_capsule, rotation=<rot 0 -90 0>}]
 root [{shapeType=Capsule, length=0.505, radius=0.505, shapeName=root_capsule, rotation=<rot 0 -90 0>}]
}
constraints [
 {bone1Name=dyn_spike_1, bone2Name=body, swing1LimitDegrees=11, swing2LimitDegrees=14, twistLimitDegrees=1}
 {bone1Name=dyn_spike_2, bone2Name=body, swing1LimitDegrees=11, swing2LimitDegrees=14, twistLimitDegrees=1}
 {bone1Name=dyn_spike_3, bone2Name=body, swing1LimitDegrees=11, swing2LimitDegrees=14, twistLimitDegrees=1}
 {bone1Name=dyn_spike_4, bone2Name=body, swing1LimitDegrees=11, swing2LimitDegrees=14, twistLimitDegrees=1}
 {
  bone1Name=body
  bone2Name=pack
  swing1LimitDegrees=45
  swing1Motion=Locked
  swing2LimitDegrees=45
  swing2Motion=Locked
  twistLimitDegrees=45
  twistMotion=Locked
 }
 {
  bone1Name=pack
  bone2Name=root
  swing1LimitDegrees=45
  swing1Motion=Locked
  swing2LimitDegrees=45
  swing2Motion=Locked
  twistLimitDegrees=45
  twistMotion=Locked
 }
]
