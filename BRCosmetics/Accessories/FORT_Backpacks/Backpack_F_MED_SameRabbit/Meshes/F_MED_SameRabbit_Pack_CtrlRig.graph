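# Control Rig for the SameRabbit back bling, a creature skeleton with jaw rings of teeth, eyes, tail and spikes, that adds procedural tail motion.
# Begin Execution drives a single Chain Harmonics node rooted at dyn_tail_1.
# Only its wave mode is enabled, shaped by a cubic curve, with pendulum and reach off and debug drawing left on.
# All other bones in the hierarchy are untouched by the graph.
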ControlRigBlueprint /BRCosmetics/Accessories/FORT_Backpacks/Backpack_F_MED_SameRabbit/Meshes/F_MED_SameRabbit_Pack_CtrlRig
graphs [
 RigVMFunctionLibrary {links [], nodes []}
 RigVMModel {nodes [
   RigUnit_BeginExecution = &RigUnit_BeginExecution()
   /Script/ControlRig.RigUnit_ChainHarmonicsPerItem(
    ChainRoot=(Type=Bone,Name="dyn_tail_1")
    DrawWorldOffset=(Rotation=(X=0,Y=0,Z=0,W=1),Translation=(X=0,Y=0,Z=0),Scale3D=(X=1,Y=1,Z=1))
    Pendulum=(bEnabled=False,PendulumStiffness=2,PendulumGravity=(X=0,Y=0,Z=0),PendulumBlend=0.75,PendulumDrag=0.98,PendulumMinimum=0,PendulumMaximum=0.1,PendulumEase=Linear,UnwindAxis=(X=0,Y=1,Z=0),UnwindMinimum=0.2,UnwindMaximum=0.05)
    Reach=(bEnabled=false,ReachTarget=(X=0,Y=0,Z=0),ReachAxis=(X=1,Y=0,Z=0),ReachMinimum=0,ReachMaximum=0,ReachEase=Linear)
    Speed=(X=3,Y=4,Z=2)
    Wave=(bEnabled=True,WaveFrequency=(X=1,Y=-3.5,Z=1),WaveAmplitude=(X=0.25,Y=0.5,Z=0.1),WaveOffset=(X=0,Y=1,Z=1),WaveNoise=(X=0,Y=0,Z=0),WaveMinimum=0,WaveMaximum=1.2,WaveEase=Linear)
    WaveCurve=(EditorCurveData=(Keys=((InterpMode=RCIM_Cubic,TangentMode=RCTM_User,ArriveTangent=6.5527,LeaveTangent=3.1142),(InterpMode=RCIM_Cubic,Time=1,Value=1,LeaveTangent=2.9368))))
    bDrawDebug=True
    ExecutePin=@RigUnit_BeginExecution.ExecutePin
   ) {is=struct RigVMUnitNode}
  ]}
]
hierarchy [
 root {loc=<0 0 0>, rot=<0 0 0 1>}
 pack {parent=root, loc=<0 -0.7316 0>, rot=<0 0 0 1>}
 body {parent=pack, loc=<0.0877 -5.905 -7.6516>, rot=<0.477 0.4873 0.5228 -0.5116>}
 dyn_neck_1 {parent=body, loc=<4.2646 0 0>, rot=<0.0009 -0.0958 -0.0117 -0.9953>}
 dyn_neck_2 {parent=dyn_neck_1, loc=<2.0136 0 0>, rot=<0 0 0 1>}
 dyn_head {parent=dyn_neck_2, loc=<5.7435 -0.0116 -0.24>, rot=<0 0.0207 0.001 -0.9998>}
 dyn_Big_jaw_rotate {parent=dyn_head, loc=<1.2475 0 -3.8146>, rot=<-0.4419 -0.4419 -0.552 -0.552>}
 dyn_Big_jaw_up {parent=dyn_Big_jaw_rotate, loc=<0 0 0>, rot=<0 0 0 1>}
 dyn_Big_jaw_down {parent=dyn_Big_jaw_rotate, loc=<0 0 0>, rot=<0 0 0 1>}
 dyn_Lit_ring {parent=dyn_Big_jaw_rotate, loc=<0.0133 -1.908 0>, rot=<0 0 0 1>}
 dyn_Lit_ring_up {parent=dyn_Lit_ring, loc=<0 0 0>, rot=<0 0 0 1>}
 dyn_Lit_teeth_1 {parent=dyn_Lit_ring_up, loc=<1.6343 -1.1512 1.1743>, rot=<-0.2362 0.1459 0.6003 -0.7501>}
 dyn_Lit_teeth_2 {parent=dyn_Lit_ring_up, loc=<-0.3012 -1.59 2.0106>, rot=<0.5622 -0.4982 -0.4795 0.4536>}
 dyn_Lit_teeth_3 {parent=dyn_Lit_ring_up, loc=<-1.5475 -1.2464 1.2712>, rot=<0.7315 -0.5972 -0.3052 0.123>}
 dyn_Lit_teeth_4 {parent=dyn_Lit_ring_up, loc=<-2.1334 -0.9709 0.4477>, rot=<0.7855 -0.6018 -0.1336 0.055>}
 dyn_Lit_ring_down {parent=dyn_Lit_ring, loc=<0 0 0>, rot=<0 0 0 1>}
 dyn_Lit_teeth_7 {parent=dyn_Lit_ring_down, loc=<1.1138 -1.5575 -1.4801>, rot=<0.547 -0.1253 0.6886 -0.4592>}
 dyn_Lit_teeth_6 {parent=dyn_Lit_ring_down, loc=<-0.4825 -1.6573 -1.7384>, rot=<0.632 -0.3759 0.6336 -0.2405>}
 dyn_Lit_teeth_5 {parent=dyn_Lit_ring_down, loc=<-1.4788 -1.5848 -1.3363>, rot=<0.7431 -0.4749 0.4285 -0.1968>}
 dyn_Mid_ring {parent=dyn_Big_jaw_rotate, loc=<-0.0121 -1.1966 0>, rot=<0 0 0 1>}
 dyn_Mid_ring_down {parent=dyn_Mid_ring, loc=<0 0 0>, rot=<0 0 0 1>}
 dyn_Mid_teeth_7 {parent=dyn_Mid_ring_down, loc=<-3.2257 -1.8616 -1.287>, rot=<0.7524 -0.5879 0.2631 -0.1385>}
 dyn_Mid_teeth_8 {parent=dyn_Mid_ring_down, loc=<-2.0616 -2.4272 -2.561>, rot=<0.6518 -0.5298 0.5132 -0.1763>}
 dyn_Mid_teeth_9 {parent=dyn_Mid_ring_down, loc=<1.6684 -2.354 -2.9826>, rot=<0.4184 -0.3697 0.5436 -0.6268>}
 dyn_Mid_teeth_10 {parent=dyn_Mid_ring_down, loc=<2.7338 -1.9875 -2.0949>, rot=<0.318 -0.0887 0.6973 -0.6362>}
 dyn_Mid_teeth_11 {parent=dyn_Mid_ring_down, loc=<3.2256 -1.6323 -1.2592>, rot=<0.2096 -0.0729 0.7283 -0.6484>}
 dyn_Mid_ring_up {parent=dyn_Mid_ring, loc=<0 0 0>, rot=<0 0 0 1>}
 dyn_Mid_teeth_2 {parent=dyn_Mid_ring_up, loc=<1.2636 -2.1567 3.8633>, rot=<-0.3792 0.4106 0.3689 -0.7426>}
 dyn_Mid_teeth_3 {parent=dyn_Mid_ring_up, loc=<1.4659 -2.2192 2.9814>, rot=<-0.4386 0.236 0.6806 -0.5373>}
 dyn_Mid_teeth_1 {parent=dyn_Mid_ring_up, loc=<2.4674 -1.9928 3.0137>, rot=<-0.3134 0.2662 0.5053 -0.7587>}
 dyn_Mid_teeth_4 {parent=dyn_Mid_ring_up, loc=<-0.4105 -2.466 3.5477>, rot=<0.6691 -0.3515 -0.5718 0.319>}
 dyn_Mid_teeth_5 {parent=dyn_Mid_ring_up, loc=<-3.014 -2.0538 1.9908>, rot=<0.7009 -0.6103 -0.3632 0.0668>}
 dyn_Mid_teeth_6 {parent=dyn_Mid_ring_up, loc=<-3.8918 -1.8587 1.084>, rot=<0.6062 -0.7806 -0.129 0.0806>}
 dyn_Mid_teeth_12 {parent=dyn_Mid_ring_up, loc=<3.4297 -1.608 0.8435>, rot=<-0.2387 -0.033 0.6116 -0.7536>}
 dyn_Big_ring {parent=dyn_Big_jaw_rotate, loc=<-0.0121 -1.1664 0>, rot=<-0.0083 0.0089 0.7071 -0.7071>}
 dyn_Big_ring_up {parent=dyn_Big_ring, loc=<0 0 0>, rot=<0.0083 -0.0089 -0.7071 -0.7071>}
 dyn_Big_teeth_4 {parent=dyn_Big_ring_up, loc=<-5.15 -0.7833 1.0486>, rot=<0.7502 -0.6611 -0.0112 0.0006>}
 dyn_Big_teeth_3 {parent=dyn_Big_ring_up, loc=<-3.5153 -1.7175 3.7699>, rot=<0.7106 -0.58 -0.3344 0.2163>}
 dyn_Big_teeth_2 {parent=dyn_Big_ring_up, loc=<-0.2641 -1.8503 5.3044>, rot=<0.5536 -0.4338 -0.515 0.49>}
 dyn_Big_teeth_1 {parent=dyn_Big_ring_up, loc=<2.086 -1.6172 4.99>, rot=<-0.3733 0.2906 0.6379 -0.6077>}
 dyn_Big_teeth_13 {parent=dyn_Big_ring_up, loc=<4.2475 -1.116 2.8326>, rot=<-0.2307 0.1331 0.6527 -0.7093>}
 dyn_Big_teeth_12 {parent=dyn_Big_ring_up, loc=<5.0095 -0.5094 0.7001>, rot=<-0.0351 0.0141 0.7444 -0.6667>}
 dyn_Big_ring_down {parent=dyn_Big_ring, loc=<0 0 0>, rot=<0.0083 -0.0089 -0.7071 -0.7071>}
 dyn_Big_teeth_5 {parent=dyn_Big_ring_down, loc=<-4.0378 -1.6192 -2.7774>, rot=<0.7413 -0.5497 0.3528 -0.1543>}
 dyn_Big_teeth_6 {parent=dyn_Big_ring_down, loc=<-2.4561 -2.0735 -4.1986>, rot=<0.67 -0.4518 0.4566 -0.372>}
 dyn_Big_teeth_7 {parent=dyn_Big_ring_down, loc=<-1.28 -2.019 -4.3243>, rot=<0.591 -0.3913 0.5941 -0.3803>}
 dyn_Big_teeth_8 {parent=dyn_Big_ring_down, loc=<-0.3128 -2.1669 -4.7508>, rot=<0.622 -0.4411 0.5328 -0.3669>}
 dyn_Big_teeth_9 {parent=dyn_Big_ring_down, loc=<2.217 -1.7745 -4.7413>, rot=<0.4096 -0.3733 0.6126 -0.5636>}
 dyn_Big_teeth_10 {parent=dyn_Big_ring_down, loc=<3.5647 -1.3588 -4.0849>, rot=<0.2762 -0.3952 0.3834 -0.7877>}
 dyn_Big_teeth_11 {parent=dyn_Big_ring_down, loc=<4.7634 -1.1256 -2.0624>, rot=<0.2599 -0.0559 0.5993 -0.7551>}
 dyn_innerMouth {parent=dyn_Big_jaw_rotate, loc=<0 -0.7198 -0.3035>, rot=<0 0 0 1>}
 dyn_eye_base_r {parent=dyn_head, loc=<-0.0904 -6.9966 -0.2975>, rot=<-0.2778 -0.4967 0.5397 -0.6204>}
 dyn_eye_R {parent=dyn_eye_base_r, loc=<0 0 0>, rot=<0 0 0 1>}
 dyn_eye_R_end {parent=dyn_eye_R, loc=<1.9414 0 0>, rot=<0 0 0 1>}
 dyn_eyer_R {parent=dyn_eye_base_r, loc=<0 0 0>, rot=<0 0 0 1>}
 dyn_eye_base_l {parent=dyn_head, loc=<-0.0213 6.9966 -0.2821>, rot=<0.3319 -0.4467 -0.5818 -0.5932>}
 dyn_eye_L {parent=dyn_eye_base_l, loc=<0 0 0>, rot=<0 0 0 1>}
 dyn_eye_L_end {parent=dyn_eye_L, loc=<1.9414 0 0>, rot=<0 0 0 1>}
 dyn_eyer_L {parent=dyn_eye_base_l, loc=<0 0 0>, rot=<0 0 0 1>}
 dyn_hand_1 {parent=body, loc=<2.3145 -0.0874 -3.2591>, rot=<-0.3838 0.5939 0.602 0.3709>}
 dyn_hand_2 {parent=dyn_hand_1, loc=<6.0852 -0.5 0>, rot=<-0.4975 0.5025 -0.4975 -0.5025>}
 dyn_tail_1 {parent=body, loc=<-7.2422 -0.2587 -1.6935>, rot=<0.0128 -0.9894 -0.0017 -0.1446>}
 dyn_tail_2 {parent=dyn_tail_1, loc=<2.8561 0 0>, rot=<-0.0009 -0.0463 -0.0049 -0.9989>}
 dyn_tail_3 {parent=dyn_tail_2, loc=<2.767 0 0>, rot=<-0.0013 -0.1411 0.0881 -0.9861>}
 dyn_tail_4 {parent=dyn_tail_3, loc=<4.003 -0.1059 -0.043>, rot=<-0.0237 -0.307 0.0467 -0.9503>}
 dyn_tail_end {parent=dyn_tail_4, loc=<2.8865 0 0>, rot=<0 0 0 1>}
 dyn_spike_3 {parent=body, loc=<-4.7809 -5.2557 -3.0736>, rot=<-0.4158 0.5917 0.6852 -0.0864>}
 dyn_spike_3_end {parent=dyn_spike_3, loc=<5.4 0 0>, rot=<0 0 0 1>}
 dyn_spike_2 {parent=body, loc=<-4.7824 -5.3424 2.2065>, rot=<0.1021 -0.7159 -0.5723 0.3866>}
 dyn_spike_2_end {parent=dyn_spike_2, loc=<4.6 0 0>, rot=<0 0 0 1>}
 dyn_spike_1 {parent=body, loc=<-5.0039 4.9566 2.2065>, rot=<0.3784 0.5774 0.7175 0.0925>}
 dyn_spike_1_end {parent=dyn_spike_1, loc=<4.9 0 0>, rot=<0 0 0 1>}
 dyn_spike_4 {parent=body, loc=<-4.9984 4.8582 -3.0736>, rot=<-0.0725 -0.6868 -0.6003 -0.4033>}
 dyn_spike_4_end {parent=dyn_spike_4, loc=<5.5 0 0>, rot=<0 0 0 1>}
 body_breath {parent=pack, loc=<0.0877 -5.905 -7.6516>, rot=<0.477 0.4873 0.5228 -0.5116>}
]
