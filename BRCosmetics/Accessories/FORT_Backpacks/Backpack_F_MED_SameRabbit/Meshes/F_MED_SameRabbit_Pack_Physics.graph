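# Physics asset for the SameRabbit backpack cosmetic, apparently for secondary motion on a dangling hand piece.
# Capsule bodies on root, pack, body, dyn_hand_1 and dyn_hand_2 are joined in a single constraint chain.
# All joints lock swing and twist except dyn_hand_1 to body, which has narrow swing and very tight twist limit angles, so both hand segments move as one unit.
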PhysicsAsset /BRCosmetics/Accessories/FORT_Backpacks/Backpack_F_MED_SameRabbit/Meshes/F_MED_SameRabbit_Pack_Physics
bodies {
 body [{shapeType=Capsule, length=1, radius=1}]
 dyn_hand_1 [{shapeType=Capsule, length=1, radius=1}]
 dyn_hand_2 [{shapeType=Capsule, length=1, radius=1}]
 pack [{shapeType=Capsule, length=0.505, radius=0.505, shapeName=pack_capsule, rotation=<rot 0 -90 0>}]
 root [{shapeType=Capsule, length=0.505, radius=0.505, shapeName=root_capsule, rotation=<rot 0 -90 0>}]
}
constraints [
 {bone1Name=dyn_hand_1, bone2Name=body, swing1LimitDegrees=20, swing2LimitDegrees=7, twistLimitDegrees=1}
 {
  bone1Name=dyn_hand_2
  bone2Name=dyn_hand_1
  swing1LimitDegrees=45
  swing1Motion=Locked
  swing2LimitDegrees=45
  swing2Motion=Locked
  twistLimitDegrees=45
  twistMotion=Locked
 }
 {
  bone1Name=body
  bone2Name=pack
  swing1LimitDegrees=45
  swing1Motion=Locked
  swing2LimitDegrees=45
  swing2Motion=Locked
  twistLimitDegrees=45
  twistMotion=Locked
 }
 {
  bone1Name=pack
  bone2Name=root
  swing1LimitDegrees=45
  swing1Motion=Locked
  swing2LimitDegrees=45
  swing2Motion=Locked
  twistLimitDegrees=45
  twistMotion=Locked
 }
]
