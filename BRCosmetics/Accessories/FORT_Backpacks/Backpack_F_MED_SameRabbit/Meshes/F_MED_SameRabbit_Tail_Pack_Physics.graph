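# Physics asset for the SameRabbit tail backpack cosmetic, defining collision bodies and joints that allow slight movement in the tail.
# Each bone has a single capsule, with small placeholder-sized ones on root, pack and body and fitted ones along the four dyn_tail bones.
# Root, pack and body are chained with angularly locked constraints as an anchor.
# The tail links hang off body in sequence with very tight swing and twist limits, keeping the motion subtle.
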PhysicsAsset /BRCosmetics/Accessories/FORT_Backpacks/Backpack_F_MED_SameRabbit/Meshes/F_MED_SameRabbit_Tail_Pack_Physics
bodies {
 body [{shapeType=Capsule, length=1, radius=1}]
 dyn_tail_1 [{
   shapeType=Capsule
   length=2.8614
   radius=2.7405
   shapeName=dyn_tail_1_capsule
   center=<1.7796 -0.0566 0.0241>
   rotation=<rot -89.581 56.7683 -146.769>
  }]
 dyn_tail_2 [{
   shapeType=Capsule
   length=3.55
   radius=2.0065
   shapeName=dyn_tail_2_capsule
   center=<1.9008 -0.1231 0.2876>
   rotation=<rot 4.5686 -4.5233 -83.8344>
  }]
 dyn_tail_3 [{shapeType=Capsule, length=3.0379, radius=1.8292, shapeName=dyn_tail_3_capsule, center=<2.5248 0 0>, rotation=<rot -0.0002 0 -90>}]
 dyn_tail_4 [{
   shapeType=Capsule
   length=1.6001
   radius=1.9768
   shapeName=dyn_tail_4_capsule
   center=<1.4441 0.0708 0.3926>
   rotation=<rot -90 -168.6578 83.6578>
  }]
 pack [{shapeType=Capsule, length=0.505, radius=0.505, shapeName=pack_capsule, rotation=<rot 0 -90 0>}]
 root [{shapeType=Capsule, length=0.505, radius=0.505, shapeName=root_capsule, rotation=<rot 0 -90 0>}]
}
constraints [
 {bone1Name=dyn_tail_1, bone2Name=body, swing1LimitDegrees=4, swing2LimitDegrees=4, twistLimitDegrees=1}
 {bone1Name=dyn_tail_2, bone2Name=dyn_tail_1, swing1LimitDegrees=2, swing2LimitDegrees=2, twistLimitDegrees=1}
 {
  bone1Name=body
  bone2Name=pack
  swing1LimitDegrees=45
  swing1Motion=Locked
  swing2LimitDegrees=45
  swing2Motion=Locked
  twistLimitDegrees=45
  twistMotion=Locked
 }
 {
  bone1Name=pack
  bone2Name=root
  swing1LimitDegrees=45
  swing1Motion=Locked
  swing2LimitDegrees=45
  swing2Motion=Locked
  twistLimitDegrees=45
  twistMotion=Locked
 }
 {bone1Name=dyn_tail_3, bone2Name=dyn_tail_2, swing1LimitDegrees=3, swing2LimitDegrees=3, twistLimitDegrees=1}
 {bone1Name=dyn_tail_4, bone2Name=dyn_tail_3, swing1LimitDegrees=2, swing2LimitDegrees=2, twistLimitDegrees=1}
]
