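# Anim Blueprint for the SameRabbit backpack cosmetic, animating its head, mouth/chest, nubs and independently blinking eyes.
# A head-turn sequence takes additive layers, including per-eye blink state machines, then offsets the body bone while skydiving.
# Three chained RigidBody nodes (pack, tail, mini tentacles) add physics before a Control Rig pass.
# State notifies and randomized delays time each eye's blink, with elapsed-time checks in the update event.
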AnimBlueprint /BRCosmetics/Accessories/FORT_Backpacks/Backpack_F_MED_SameRabbit/Meshes/F_MED_SameRabbit_Pack_AnimBP
target_skeleton=&/F_MED_SameRabbit_Pack_Skeleton
graphs [
 AnimGraph {nodes [
   n1 = Sequence Player 'SameRabbit_HeadTurns_PACK'() {sequence=/BRCosmetics/Accessories/FORT_Backpacks/Backpack_F_MED_SameRabbit/Anim/SameRabbit_HeadTurns_PACK}
   n2 = "Sequence Player (additive) 'sameRabbit_MouthChest_PACK'"() {sequence=/BRCosmetics/Accessories/FORT_Backpacks/Backpack_F_MED_SameRabbit/Anim/sameRabbit_MouthChest_PACK}
   AnimGraphNode_StateMachine_2 = Left Eye Blink() {is=title AnimGraphNode_StateMachine}
   n3 = Apply Additive(
    Base=Apply Additive(
     Base=Apply Additive(Base=@n1, Additive=@n2) {alphaBoolBlend {blendOption=Linear}}
     Additive="Sequence Player (additive) 'SameRabbit_nubs_PACK'"() {sequence=/BRCosmetics/Accessories/FORT_Backpacks/Backpack_F_MED_SameRabbit/Anim/SameRabbit_nubs_PACK}
    ) {alphaBoolBlend {blendOption=Linear}}
    Additive=@AnimGraphNode_StateMachine_2
   ) {alphaBoolBlend {blendOption=Linear}}
   AnimGraphNode_StateMachine_1 = Right Eye Blink() {is=title AnimGraphNode_StateMachine}
   n4 = "Transform (Modify) Bone - Bone: body"!(
    ComponentPose=Local To Component(LocalPose=Apply Additive(Base=@n3, Additive=@AnimGraphNode_StateMachine_1) {alphaBoolBlend {blendOption=Linear}})
    bAlphaBoolEnabled:bool=Get bIsSkyDiving()
    Translation="3.116,0.000,-1.690"
    Rotation="13.424,0.000,0.000"
    Scale="1,1,1"
   ) -> (Pose) {
    alphaInputType=Bool
    rotationMode=BMM_Additive
    rotationSpace=BCS_BoneSpace
    translationMode=BMM_Additive
    translationSpace=BCS_BoneSpace
    alphaBoolBlend {blendOption=Linear}
    boneToModify {boneName=body}
    rotation=<rot 13.424 0 0>
    translation {x=3.116, z=-1.69}
   }
   n5 = RigidBody(ComponentPose=@n4) {
    bOverrideWorldGravity=true
    overridePhysicsAsset=&/F_MED_SameRabbit_Pack_Physics
    alphaBoolBlend {blendOption=Linear}
    componentAppliedLinearAccClamp=<1000 1000 1000>
    componentLinearAccScale {z=1}
    componentLinearVelScale {z=1}
    overrideWorldGravity {z=-500}
    simSpaceSettings {maxAngularAcceleration=150, maxAngularVelocity=150, maxLinearAcceleration=1, maxLinearVelocity=2, worldAlpha=1}
   }
   n6 = RigidBody(ComponentPose=@n5) {
    bOverrideWorldGravity=true
    overridePhysicsAsset=&/F_MED_SameRabbit_Tail_Pack_Physics
    alphaBoolBlend {blendOption=Linear}
    componentAppliedLinearAccClamp=<1000 1000 1000>
    componentLinearAccScale {z=1}
    componentLinearVelScale {z=1}
    overrideWorldGravity {z=-500}
    simSpaceSettings {maxAngularAcceleration=150, maxAngularVelocity=150, maxLinearAcceleration=1, maxLinearVelocity=2, worldAlpha=1}
   }
   n7 = RigidBody(ComponentPose=@n6, Alpha=0.65) {
    bOverrideWorldGravity=true
    overridePhysicsAsset=&/F_MED_SameRabbit_MiniTentacles_Pack_Physics
    alphaBoolBlend {blendOption=Linear}
    componentAppliedLinearAccClamp=<1000 1000 1000>
    componentLinearAccScale {z=1}
    componentLinearVelScale {z=1}
    overrideWorldGravity {z=-500}
    simSpaceSettings {maxAngularAcceleration=150, maxAngularVelocity=150, maxLinearAcceleration=1, maxLinearVelocity=2, worldAlpha=1}
   }
   n8 = Control Rig(Source=Component To Local(ComponentPose=@n7)) {
    alphaBoolBlend {blendOption=Linear}
    controlRigAssetReference {blueprintRigClass=&/F_MED_SameRabbit_Pack_CtrlRig.F_MED_SameRabbit_Pack_CtrlRig_C}
   }
   Output Pose(Result=@n8) {layerGroup=None}
  ]}
 Right Eye Blink {
  entry=Identity
  path=AnimGraph.AnimGraphNode_StateMachine_1.Right Eye Blink
  ids=id
  nodes [
   AnimStateNode_2 = Identity() {is=title AnimStateNode}
   AnimStateNode_3 = Blink() {
    is=title AnimStateNode
    stateEntered {guid=#1, notifyName=EnteredBlink_Right}
    stateLeft {guid=#2, notifyName=ExitedBlink_Right}
   }
   AnimStateTransitionNode_2 = AnimStateTransitionNode()
   AnimStateTransitionNode_3 = AnimStateTransitionNode() {bAutomaticRuleBasedOnSequencePlayerInState=true}
  ]
  links [
   {
    dst=Identity
    rule=AnimGraph.AnimGraphNode_StateMachine_1.Right Eye Blink.AnimStateTransitionNode_3.Transition
    src=Blink
    via=AnimStateTransitionNode_3
   }
   {
    dst=Blink
    rule=AnimGraph.AnimGraphNode_StateMachine_1.Right Eye Blink.AnimStateTransitionNode_2.Transition
    src=Identity
    via=AnimStateTransitionNode_2
   }
  ]
 }
 Identity {
  path=AnimGraph.AnimGraphNode_StateMachine_1.Right Eye Blink.AnimStateNode_2.Identity
  nodes [Output Animation Pose(Result=Additive Identity Pose())]
 }
 Blink {
  path=AnimGraph.AnimGraphNode_StateMachine_1.Right Eye Blink.AnimStateNode_3.Blink
  nodes [Output Animation Pose(Result="Sequence Player (additive) 'sameRabbit_R_EYE_PACK'"() {bLoopAnimation=false, sequence=/BRCosmetics/Accessories/FORT_Backpacks/Backpack_F_MED_SameRabbit/Anim/sameRabbit_R_EYE_PACK})]
 }
 Transition {
  path=AnimGraph.AnimGraphNode_StateMachine_1.Right Eye Blink.AnimStateTransitionNode_2.Transition
  nodes [Result(bCanEnterTransition=Get CanBlink_Right() -> (CanBlink_Right:bool))]
 }
 Transition {path=AnimGraph.AnimGraphNode_StateMachine_1.Right Eye Blink.AnimStateTransitionNode_3.Transition, nodes [Result()]}
 Left Eye Blink {
  entry=Identity
  path=AnimGraph.AnimGraphNode_StateMachine_2.Left Eye Blink
  ids=id
  nodes [
   AnimStateNode_2 = Identity()
   AnimStateNode_3 = Blink() {
    stateEntered {guid=#3, notifyName=EnteredBlink_Left}
    stateLeft {guid=#4, notifyName=ExitedBlink_Left}
   }
   AnimStateTransitionNode_4 = AnimStateTransitionNode()
   AnimStateTransitionNode_5 = AnimStateTransitionNode() {bAutomaticRuleBasedOnSequencePlayerInState=true}
  ]
  links [
   {
    dst=Identity
    rule=AnimGraph.AnimGraphNode_StateMachine_2.Left Eye Blink.AnimStateTransitionNode_5.Transition
    src=Blink
    via=AnimStateTransitionNode_5
   }
   {
    dst=Blink
    rule=AnimGraph.AnimGraphNode_StateMachine_2.Left Eye Blink.AnimStateTransitionNode_4.Transition
    src=Identity
    via=AnimStateTransitionNode_4
   }
  ]
 }
 Identity {
  path=AnimGraph.AnimGraphNode_StateMachine_2.Left Eye Blink.AnimStateNode_2.Identity
  nodes [Output Animation Pose(Result=Additive Identity Pose())]
 }
 Blink {
  path=AnimGraph.AnimGraphNode_StateMachine_2.Left Eye Blink.AnimStateNode_3.Blink
  nodes [Output Animation Pose(Result="Sequence Player (additive) 'sameRabbit_L_EYE_PACK'"() {bLoopAnimation=false, sequence=/BRCosmetics/Accessories/FORT_Backpacks/Backpack_F_MED_SameRabbit/Anim/sameRabbit_L_EYE_PACK})]
 }
 Transition {
  path=AnimGraph.AnimGraphNode_StateMachine_2.Left Eye Blink.AnimStateTransitionNode_4.Transition
  nodes [Result(bCanEnterTransition=Get CanBlink_Left() -> (CanBlink_Left:bool))]
 }
 Transition {path=AnimGraph.AnimGraphNode_StateMachine_2.Left Eye Blink.AnimStateTransitionNode_5.Transition, nodes [Result()]}
 EventGraph {nodes [
   Event BlueprintUpdateAnimation()
   > n1 = Sequence()
   n2 = GetGameTimeInSeconds()
   n3 = Get LastTimeBlink_Left() -> (LastTimeBlink_Left:double)
   > Branch(Condition="float > float"(A=float - float(A=@n2, B=@n3), B=10))
   > Set CanBlink_Left(CanBlink_Left:bool=true) -> (Output_Get:bool)
   n4 = GetGameTimeInSeconds()
   n5 = Get LastTimeBlink_Right() -> (LastTimeBlink_Right:double)
   Branch(execute=@n1.then_1, Condition="float > float"(A=float - float(A=@n4, B=@n5), B=9))
   > Set CanBlink_Right(CanBlink_Right:bool=true) -> (Output_Get:bool)
   Event BlueprintInitializeAnimation()
   > n6 = Sequence()
   > Delay(Duration=RandomFloatInRange(Min=0.5, Max=3))
   > n7 = Set CanBlink_Left(CanBlink_Left=true)
   Set LastTimeBlink_Left(execute=[@n7.then, @n9.then], LastTimeBlink_Left:double=@n2) -> (Output_Get:double)
   Delay(execute=@n6.then_1, Duration=RandomFloatInRange(Min=0.5, Max=3))
   > n8 = Set CanBlink_Right(CanBlink_Right=true)
   Set LastTimeBlink_Right(execute=[@n8.then, @n10.then], LastTimeBlink_Right:double=@n4) -> (Output_Get:double)
   AnimNotify_EnteredBlink_Left() {is=title K2Node_Event}
   > Set CanBlink_Left(CanBlink_Left=false)
   AnimNotify_EnteredBlink_Right() {is=title K2Node_Event}
   > Set CanBlink_Right(CanBlink_Right=false)
   AnimNotify_ExitedBlink_Left() {is=title K2Node_Event}
   > Delay(Duration=RandomFloatInRange(Min=0.5, Max=5))
   > n9 = Set CanBlink_Left(CanBlink_Left=true)
   AnimNotify_ExitedBlink_Right() {is=title K2Node_Event}
   > Delay(Duration=RandomFloatInRange(Min=0.5, Max=5))
   > n10 = Set CanBlink_Right(CanBlink_Right=true)
  ]}
]
variables [
 Blinking_Left {type=boolean}
 Blinking_Right {type=boolean}
 CanBlink_Left {type=boolean}
 CanBlink_Right {type=boolean}
 LastTimeBlink_Left {type=number}
 LastTimeBlink_Right {type=number}
]
